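# Animation Blueprint for the M_MED_FumeFleeceShack male player body cosmetic, following the pose of its attached parent mesh.
# The copied pose is blended per bone with DefaultSlot from dyn_holster_MSTR down.
# In the front end, each update sets that blend weight from the Frontend_Idle_Bypass curve on the owner's front-end anim instance, letting custom idles drive the holster bones.
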AnimBlueprint /BRCosmetics/Characters/Player/Male/Medium/Bodies/M_MED_FumeFleeceShack/Meshes/M_MED_FumeFleeceShack_AnimBP
target_skeleton=&/M_MED_FumeFleeceShack_Skeleton
graphs [
 AnimGraph {nodes [Output Pose(Result=Layered blend per bone(
     BasePose=Copy Pose From Mesh() {bUseAttachedParent=true}
     BlendPoses_0=Slot 'DefaultSlot'()
     BlendWeights_0=Get Custom Idle is Playing() -> (Custom Idle is Playing:double)
    ) {layerSetup [{branchFilters [{boneName=dyn_holster_MSTR}]}]}) {layerGroup=None}]}
 EventGraph {nodes [
   Event BlueprintUpdateAnimation()
   > n1 = Sequence()
   Branch(execute=@n1.then_1, Condition=Get bIsInFrontEnd())
   > Set Custom Idle is Playing(Custom Idle is Playing:double=GetCurveValue(self=GetOwnerFrontEndAnimInstance(), CurveName=Frontend_Idle_Bypass)) -> (Output_Get:double)
  ]}
]
variables [Custom Idle is Playing {type=number}]
